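# Material function that applies three Super Levels cosmetic tiers to the BengalBasher suit skin, selected by SL1/SL2/SL3 static switches.
# Each tier blends primary, secondary, accent and optional emissive layers over input material attributes using tech-art mask channels.
# T1 uses a scrolling gradient-mapped pattern, T2 panning paint strokes, T3 hue-shifting shards with fake refraction and sun inscattering.
# Pre-skinned cylindrical UVs feed the projected effects; a DiamondHeart_Sapphire switch alters emissive masking.
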MaterialFunction /BRCosmetics/Characters/Player/Male/Large/Bodies/M_LRG_BengalBasher_Suit/Skins/SL/MF_S28_SuperLevels
graph {nodes [
  n1 = NamedRerouteUsage()
  n2 = Multiply(A=NamedRerouteUsage(), B=Max(A=NamedRerouteUsage(), B=NamedRerouteUsage()))
  Add(A=@n1, B=@n2)
  n3 = NamedRerouteUsage()
  n4 = NamedRerouteUsage()
  n5 = Fresnel() {baseReflectFraction=0, exponent=0.1}
  n6 = LinearInterpolate(
   A=VectorParameter() {parameterName=SL_T3_EmissiveMaskColor, defaultValue=<rgba 1 0.374 0 1>}.RGB
   B=VectorParameter() {parameterName=SL_T3_EmissiveMaskColor2, defaultValue=<rgba 1 0.0372 0 1>}.RGB
   Alpha=Multiply(A=Power(Base=@n4) {constExponent=0.1}, B=@n5)
  )
  n7 = LinearInterpolate(
   A=ScalarParameter() {defaultValue=50, parameterName=SL_T3_EmissiveMaskBrightness}
   B=ScalarParameter() {parameterName=SL_T3_EmissiveMaskBrightnessLow}
   Alpha=@n5
  ) {constA=1}
  n8 = Power(Base=NamedRerouteUsage()) {constExponent=0.8}
  n9 = Multiply(A=NamedRerouteUsage(), B=Max(A=NamedRerouteUsage(), B=NamedRerouteUsage()))
  n10 = NamedRerouteUsage()
  n11 = BlendMaterialAttributes(
   A=GetMaterialAttributes(None=@n3) {attributeGetTypes [#1, #2]}.MaterialAttributes
   B=SetMaterialAttributes!(
    MaterialAttributes
    EmissiveColor=Multiply(A=@n6, B=@n7)
    Roughness=ScalarParameter() {defaultValue=0.61, parameterName=SL_T3_EmissiveMaskRoughness}
   ) {attributeSetTypes [#2, #3]}
   Alpha=StaticSwitchParameter(True=Multiply(A=@n8, B=@n9), False=Multiply(A=@n10, B=@n8)) {group=00_SuperLevels, parameterName=DiamondHeart_Sapphire, sortPriority=128}
  ) {vertexAttributeBlendType=UseA}
  n12 = NamedRerouteUsage()
  n13 = GetMaterialAttributes(None=@n12) {attributeGetTypes [#1, #2]}
  n14 = OneMinus(None=Fresnel() {baseReflectFraction=0, exponent=0.15})
  n15 = Multiply(
   A=LinearInterpolate(
    A=VectorParameter() {parameterName=SL_T2_EmissiveMaskColor, defaultValue=<rgba 1 0 0 1>}.RGB
    B=VectorParameter() {parameterName=SL_T2_EmissiveMaskColor2, defaultValue=<rgba 1 0.5751 0.0573 1>}.RGB
    Alpha=@n14
   )
   B=LinearInterpolate(
    A=ScalarParameter() {defaultValue=0.1, parameterName=SL_T2_EmissiveMaskBrightnessLow}
    B=ScalarParameter() {defaultValue=15, parameterName=SL_T2_EmissiveMaskBrightness}
    Alpha=@n14
   ) {constA=1}
  )
  n16 = Desaturation(None=@n13.EmissiveColor)
  n17 = NamedRerouteUsage()
  n18 = LinearInterpolate(A=@n15, B=Multiply(A=Multiply(A=@n16, B=@n17), B=@n15), Alpha=@n17)
  n19 = ScalarParameter() {defaultValue=0.61, parameterName=SL_T2_EmissiveMaskRoughness}
  n20 = StaticSwitchParameter(
   True=BlendMaterialAttributes(
    A=@n13.MaterialAttributes
    B=SetMaterialAttributes(EmissiveColor=@n18, Roughness=@n19) {attributeSetTypes [#2, #3]}
    Alpha=StaticSwitchParameter(True=@n2, False=@n1) {group=00_SuperLevels, parameterName=DiamondHeart_Sapphire, sortPriority=128}
   ) {vertexAttributeBlendType=UseA}
   False=@n12
  ) {group=00_SuperLevels, parameterName=SL_EnableEmissiveFromAlpha, sortPriority=48}
  n21 = NamedRerouteUsage()
  n22 = GetMaterialAttributes(None=@n21) {attributeGetTypes [#1, #2]}
  n23 = VectorParameter() {parameterName=SL_T1_EmissiveMaskColor2, defaultValue=<rgba 0 0 1 1>}
  n24 = VectorParameter() {parameterName=SL_T1_EmissiveMaskColor, defaultValue=<rgba 0.3 0 1 1>}
  n25 = Fresnel() {baseReflectFraction=0, exponent=0.25}
  n26 = Power(Base=NamedRerouteUsage())
  n27 = LinearInterpolate(
   A=ScalarParameter() {defaultValue=15, parameterName=SL_T1_EmissiveMaskBrightnessLow}
   B=ScalarParameter() {defaultValue=75, parameterName=SL_T1_EmissiveMaskBrightness}
   Alpha=@n26
  ) {constA=3}
  n28 = SetMaterialAttributes(
   EmissiveColor=Multiply(A=LinearInterpolate(A=@n23.RGB, B=@n24.RGB, Alpha=@n25), B=LinearInterpolate(A=@n27, Alpha=@n25) {constA=1, constB=0})
   Roughness=ScalarParameter() {defaultValue=0.18, parameterName=SL_T1_EmissiveMaskRoughness, sliderMax=1}
  ) {attributeSetTypes [#2, #3]}
  n29 = NamedRerouteDeclaration(None=FunctionInput() {inputName=DiamondHeartPattern, inputType=FunctionInput_Scalar}) {name=DiamondHeartPattern, nodeColor=<rgba 0.2891 1 0 1>}
  n30 = Max(A=NamedRerouteUsage(), B=NamedRerouteUsage())
  n31 = Multiply(
   A=Power(Base=@n26, Exp=ScalarParameter() {defaultValue=0.5, parameterName=SL_T1_EmissiveMaskBreakupExponent}) {constExponent=0.5}
   B=StaticSwitchParameter(True=Multiply(A=@n29, B=@n30), False=NamedRerouteUsage()) {group=00_SuperLevels, parameterName=DiamondHeart_Sapphire, sortPriority=128}
  )
  n32 = StaticSwitchParameter(
   True=StaticSwitchParameter(True=BlendMaterialAttributes(A=@n22.MaterialAttributes, B=@n28, Alpha=@n31) {vertexAttributeBlendType=UseA}, False=@n21) {group=00_SuperLevels, parameterName=SL_EnableEmissiveFromAlpha, sortPriority=48}
   False=NamedRerouteDeclaration(None=FunctionInput() {inputType=FunctionInput_MaterialAttributes}) {name=InMatAttr, nodeColor=<rgba 1 0 0.2344 1>}
  ) {defaultValue=true, group=00_SuperLevels, parameterName=SL1}
  FunctionOutput(None=StaticSwitchParameter(
    True=StaticSwitchParameter(True=@n11, False=@n3) {group=00_SuperLevels, parameterName=SL_EnableEmissiveFromAlpha, sortPriority=48}
    False=StaticSwitchParameter(True=@n20, False=@n32) {group=00_SuperLevels, parameterName=SL2}
   ) {group=00_SuperLevels, parameterName=SL3})
  n33 = Normalize(VectorInput=WorldPosition() {worldPositionShaderOffset=WPT_CameraRelative}.XYZ)
  n34 = Add(
   A=PreSkinnedPosition()
   B=Multiply(
    A=Transform(None=@n33) {transformSourceType=TRANSFORMSOURCE_World, transformType=TRANSFORM_Local}
    B=ScalarParameter() {defaultValue=-5, group=00_SuperLevels_T1, parameterName=SL_PositionOffset}
   ) {constB=100}
  )
  n35 = Divide(
   A=NamedRerouteDeclaration(None=VertexInterpolator(VS=@n34).PS) {name=SL_PreskinnedLocalPos, nodeColor=<rgba 0 1 0.0859 1>}
   B=NamedRerouteUsage()
  )
  n36 = Normalize(VectorInput=Transform(None=Constant3Vector() {constant=<rgba 1 0 0 0>}) {transformSourceType=TRANSFORMSOURCE_View})
  NamedRerouteDeclaration(None=DDX(Value=DotProduct(A=@n35, B=@n36))) {name=CylDDX, nodeColor=<rgba 0 1 0.9063 1>}
  NamedRerouteDeclaration(None=DDY(Value=ComponentMask(None=@n35) {b=true})) {name=CylDDY, nodeColor=<rgba 1 0 0.9609 1>}
  n37 = NamedRerouteUsage()
  n38 = Time()
  n39 = ScalarParameter() {defaultValue=6, group=00_SuperLevels_T3, parameterName=SL_T3_HueShiftPeriod}
  n40 = Multiply(A=TextureCoordinate(), B=ScalarParameter() {defaultValue=1, parameterName=SL_T3_TextureScale})
  n41 = TextureSampleParameter2D(UVs=@n40) {
   group=00_SuperLevels_T3
   parameterName=SL3_Color
   texture=/BRCosmetics/Characters/Player/Male/Large/Bodies/M_LRG_BengalBasher_Suit/Skins/SL/Textures/T_S28_SL3_D
  }
  n42 = Sine(None=Subtract(A=Divide(A=@n38, B=@n39) {constB=8}, B=@n41.B))
  n43 = ScalarParameter() {defaultValue=0.05, group=00_SuperLevels_T3, parameterName=SL_T3_HueShiftRange}
  n44 = Desaturation(
   None=MaterialFunctionCall(Hue Shift Percentage=Multiply(A=@n42, B=@n43), Texture=@n41.RGB) {materialFunction=&HueShift}.Result
   Fraction=ScalarParameter() {defaultValue=0.05, group=00_SuperLevels_T3, parameterName=SL_T3_Desaturate}
  )
  n45 = VectorParameter() {group=00_SuperLevels_T3, parameterName=SL_T3_Tint, channelNames {a=Tint Amount, b="", g="", r=""}, defaultValue=<rgba 1 0.3966 0.2564 0.2>}
  n46 = LinearInterpolate(
   A=@n44
   B=MaterialFunctionCall(Base=@n44, Blend=@n45) {materialFunction=&Blend_Overlay}.Result
   Alpha=@n45.Tint Amount
  )
  n47 = GetMaterialAttributes(None=NamedRerouteUsage()) {attributeGetTypes [#1, #4, #2]}
  n48 = TextureSampleParameter2D(UVs=@n40) {
   group=00_SuperLevels_T3
   parameterName=SL3_Normal
   samplerType=SAMPLERTYPE_Normal
   texture=/BRCosmetics/Characters/Player/Male/Large/Bodies/M_LRG_BengalBasher_Suit/Skins/SL/Textures/T_S28_SL3_N
  }
  n49 = LinearInterpolate(
   A=Constant3Vector() {constant=<rgba 0 0 1 0>}
   B=Multiply(A=@n48.RGB, B=Constant3Vector() {constant=<rgba 1 1 1 0>})
   Alpha=ScalarParameter() {defaultValue=0.08, group=00_SuperLevels_T3, parameterName=SL_T3_GlassNormalIntensity, sliderMax=1}
  )
  n50 = ComponentMask(None=NamedRerouteDeclaration(None=MaterialFunctionCall(BaseNormal=@n47.Normal, AdditionalNormal=@n49) {materialFunction=&BlendAngleCorrectedNormals}.Result) {name=T3_BlendedNormal, nodeColor=<rgba 0.4766 1 0 1>}) {g=true, r=true}
  n51 = ScalarParameter() {defaultValue=0.2, group=00_SuperLevels_T3, parameterName=SL_T3_RefractIntensity}
  n52 = Multiply(
   A=MaterialFunctionCall!(Vector=CameraVectorWS()) {materialFunction=/Engine/Functions/Engine_MaterialFunctions01/Coordinates/LongLatToUV}.Result
   B=Multiply(A=Constant2Vector() {g=1, r=2}, B=ScalarParameter() {defaultValue=2, group=00_SuperLevels_T3, parameterName=SL_T3_RefractInnerUVScale})
  )
  n53 = TextureCoordinate()
  n54 = TextureSample!(UVs=Add(A=Multiply(A=@n50, B=@n51), B=@n52), Tex, "DDX(UVs)"=DDX(Value=@n53), "DDY(UVs)"=DDY(Value=@n53), Apply View MipBias) {
   mipValueMode=TMVM_Derivative
   samplerSource=SSM_Wrap_WorldGroupSettings
   samplerType=SAMPLERTYPE_LinearColor
   texture=/BRCosmetics/Characters/Player/Male/Large/Bodies/M_LRG_BengalBasher_Suit/Skins/SL/Textures/T_S28_SL1_BG_A
  }
  n55 = ScalarParameter() {defaultValue=0.7, parameterName=SL_T3_InnerParallaxStrength}
  n56 = TextureSampleParameter2D(UVs=@n40) {
   group=00_SuperLevels_T3
   parameterName=SL3_Masks
   samplerType=SAMPLERTYPE_LinearColor
   texture=/BRCosmetics/Characters/Player/Male/Large/Bodies/M_LRG_BengalBasher_Suit/Skins/SL/Textures/T_S28_SL3_M
  }
  n57 = Subtract(A=@n56.G, B=ScalarParameter() {defaultValue=0.48, group=00_SuperLevels_T3, parameterName=SL_T3_WeldDFTreshold, sliderMax=0.6, sliderMin=0.4})
  n58 = Saturate(None=Multiply(A=NamedRerouteDeclaration(None=@n57) {name=SL_T3_ShardDF, nodeColor=<rgba 1 0.8438 0 1>}) {constB=50})
  n59 = Multiply(A=Multiply(A=@n46, B=LinearInterpolate(B=@n54.R, Alpha=@n55) {constA=1}), B=@n58)
  n60 = OneMinus(None=Power(Base=Fresnel() {baseReflectFraction=0, exponent=0.08}) {constExponent=3})
  n61 = Multiply(A=@n59, B=@n60)
  n62 = Multiply(A=ScalarParameter() {defaultValue=0.5, parameterName=T3_BaseColorIntensity}, B=@n61)
  n63 = NamedRerouteUsage()
  n64 = Power(Base=Saturate(None=Multiply(A=@n58) {constB=5})) {constExponent=4}
  n65 = Fresnel(
   ExponentIn=ScalarParameter() {defaultValue=5, group=00_SuperLevels_T3, parameterName=T3_FresnelEXP}
   BaseReflectFractionIn=ScalarParameter() {group=00_SuperLevels_T3, parameterName=T3_FresnelBase}
  )
  n66 = Multiply(A=Multiply(A=@n45, B=ScalarParameter() {defaultValue=2, parameterName=SL_T3_EmissiveBaseAmount, sliderMax=1}) {constB=0.1}, B=@n60)
  n67 = VectorParameter() {group=RENDER_Overrides, parameterName=RENDER_FakeSunlightDirection, defaultValue=<rgba -0.5 -0.5 -0.3 0>}
  n68 = StaticSwitchParameter(True=Normalize(VectorInput=Transform(None=@n67.RGB) {transformSourceType=TRANSFORMSOURCE_Local}), False=SkyAtmosphereLightDirection()) {group=RENDER_Overrides, parameterName=RENDER_FakeDirectionalLightParams, sortPriority=1}
  n69 = Transform(None=@n49)
  n70 = Multiply(A=DotProduct(
    A=MaterialFunctionCall!(Ray Direction=@n68, Surface Normal=@n69, Refractive Index Origin, Refractive Index Target) {materialFunction=/Engine/Functions/Engine_MaterialFunctions01/Vectors/Refract}.Vector
    B=CameraVectorWS()
   )) {constB=-1}
  n71 = VectorParameter() {group=RENDER_Overrides, parameterName=RENDER_FakeLightIlluminanceOnGround, defaultValue=<rgba 7.2 7 8 0>}
  n72 = StaticSwitchParameter(True=@n71.RGB, False=AtmosphericLightColor()) {group=RENDER_Overrides, parameterName=RENDER_FakeDirectionalLightParams, sortPriority=1}
  n73 = Desaturation(None=@n72)
  n74 = VectorParameter() {parameterName=SL_T3_ColorOverride, defaultValue=<rgba 1 0.4039 0.0549 1>}
  n75 = Multiply(
   A=Power(
    Base=Saturate(None=@n70)
    Exp=ScalarParameter() {defaultValue=4, group=00_SuperLevels_T3, parameterName=SL_T3_SunDirectionalInscatteringExponent, sliderMax=10}
   ) {constExponent=3}
   B=LinearInterpolate(A=@n72, B=Multiply(A=@n73, B=@n74.RGB), Alpha=ScalarParameter() {defaultValue=0.6, parameterName=SL_T3_ColorOverrideAmount, sliderMax=1})
  )
  n76 = Add(A=Multiply(A=Multiply(A=@n65) {constB=2}, B=@n59), B=Multiply(A=@n61, B=Add(A=@n66, B=@n75) {constB=0.05}))
  n77 = ScalarParameter() {defaultValue=2, group=00_SuperLevels_T3, parameterName=T3_EmissiveBrightness}
  n78 = SetMaterialAttributes!(
   MaterialAttributes
   BaseColor=LinearInterpolate(A=ScalarParameter() {defaultValue=0.11, parameterName=SL_T3_PrimarySeamBrightness}, B=Multiply(A=@n62, B=@n63), Alpha=@n64) {constA=0.08}
   ShadingModel=ShadingModel()
   EmissiveColor=Multiply(A=NamedRerouteUsage(), B=Multiply(A=@n76, B=@n77))
   Normal=NamedRerouteUsage()
   Roughness=LinearInterpolate(
    A=ScalarParameter() {defaultValue=0.36, group=00_SuperLevels_T3, parameterName=SL_T3_PrimarySeamRoughness, sliderMax=1}
    B=ScalarParameter() {defaultValue=0.1328, group=00_SuperLevels_T3, parameterName=SL_T3_PrimaryRoughness, sliderMax=1}
    Alpha=@n58
   ) {constA=0.3}
   Metallic=OneMinus(None=@n64)
  ) {attributeSetTypes [
    #1
    #5
    #2
    #4
    #3
    #6
   ]}
  n79 = StaticSwitchParameter(True=@n37, False=BlendMaterialAttributes(A=@n37, B=@n78, Alpha=NamedRerouteUsage())) {group=00_SuperLevels, parameterName=SL_DisablePrimaryLayer, sortPriority=40}
  n80 = VectorParameter() {group=00_SuperLevels_T3, parameterName=SL_T3_SecondaryTintColor, defaultValue=<rgba 0.0117 0.0153 0.026 1>}
  n81 = VectorParameter() {group=00_SuperLevels_T3, parameterName=SL_T3_SecondaryPatternColor, defaultValue=<rgba 0.2127 0.3667 0.5938 1>}
  n82 = Time()
  n83 = TextureSample!(
   UVs=Multiply(A=TextureCoordinate(), B=ScalarParameter() {defaultValue=4, group=00_SuperLevels_T3, parameterName=SL_T3_SecondaryUVScale})
   Tex
   Apply View MipBias
  ) {
   samplerSource=SSM_Wrap_WorldGroupSettings
   samplerType=SAMPLERTYPE_LinearColor
   texture=/Game/Characters/Player/Male/Large/Bodies/M_LRG_BengalBasher_Suit/Skins/SL/Textures/T_S28_SL3_DF
  }
  n84 = Frac(None=Multiply(A=ScalarParameter() {defaultValue=0.3, group=00_SuperLevels_T3, parameterName=SL_T3_SecondaryPatternSpeed}, B=Subtract(A=@n82, B=@n83.G)))
  n85 = SphereMask(A=@n84, B=Constant() {r=0.5}) {attenuationRadius=0.5, hardnessPercent=0}
  n86 = SphereMask(
   A=SphereMask(A=LinearInterpolate(Alpha=@n84) {constA=0.9, constB=0.65}, B=@n83.R) {attenuationRadius=0.125, hardnessPercent=0}
   B=Constant() {r=0.35}
  ) {attenuationRadius=0.3, hardnessPercent=0}
  n87 = Add(
   A=Multiply(
    A=Fresnel() {baseReflectFraction=0}
    B=VectorParameter() {group=00_SuperLevels_T3, parameterName=SL_FresnelColor, defaultValue=<rgba 0.6094 2.0248 3 1>}.RGB
   )
   B=LinearInterpolate(A=@n80.RGB, B=@n81.RGB, Alpha=Multiply(A=@n85, B=@n86))
  )
  n88 = ShadingModel()
  n89 = StaticSwitchParameter(
   True=@n79
   False=BlendMaterialAttributes(
    A=@n79
    B=SetMaterialAttributes!(MaterialAttributes=@n79, BaseColor=@n87, ShadingModel=@n88) {attributeSetTypes [#1, #5]}
    Alpha=NamedRerouteUsage()
   )
  ) {group=00_SuperLevels, parameterName=SL_DisableSecondaryLayer, sortPriority=41}
  n90 = GetMaterialAttributes(None=@n89) {attributeGetTypes [#1]}
  n91 = Constant() {r=1}
  n92 = SetMaterialAttributes!(
   MaterialAttributes=@n90.MaterialAttributes
   BaseColor=Multiply(
    A=Desaturation(None=@n90.BaseColor, Fraction=@n91)
    B=VectorParameter() {parameterName=SL_T3_AccentColor, defaultValue=<rgba 0.5535 0.5678 0.5938 1>}.RGB
   )
   Metallic=Constant() {r=1}
   Roughness=ScalarParameter() {defaultValue=0.22, parameterName=SL_T3_AccentRoughness}
  ) {attributeSetTypes [#1, #6, #3]}
  n93 = NamedRerouteUsage()
  NamedRerouteDeclaration(None=StaticSwitchParameter(True=@n89, False=BlendMaterialAttributes(A=@n89, B=@n92, Alpha=@n93)) {group=00_SuperLevels, parameterName=SL_DisableAccentLayer, sortPriority=42}) {name=T3 MA, nodeColor=<rgba 0 0.4141 1 1>}
  n94 = TextureSampleParameter2D() {
   group=05_Tech Art Masks
   parameterName=SL_TechArtFX_Mask
   samplerSource=SSM_Wrap_WorldGroupSettings
   samplerType=SAMPLERTYPE_Masks
   texture=/Engine/EngineMaterials/DefaultDiffuse_TC_Masks
  }
  NamedRerouteDeclaration(None=@n94.RGBA) {name=SLMask_RGBA, nodeColor=<rgba 0.2656 1 0 1>}
  NamedRerouteDeclaration(None=@n94.R) {name=SLMask_Primary, nodeColor=<rgba 0 0.1094 1 1>}
  NamedRerouteDeclaration(None=@n94.G) {name=SLMask_Accent, nodeColor=<rgba 0 0.4375 1 1>}
  NamedRerouteDeclaration(None=@n94.B) {name=SLMask_Secondary, nodeColor=<rgba 1 0 0.3516 1>}
  NamedRerouteDeclaration(None=@n94.A) {name=SLMask_Emissive, nodeColor=<rgba 1 0.0938 0 1>}
  n95 = GetMaterialAttributes(None=NamedRerouteUsage()) {attributeGetTypes [#4]}
  n96 = VectorParameter() {group=00_SuperLevels_T2, parameterName=Paint_Color5, defaultValue=<rgba 0.1891 0.2881 0.4375 1>}
  n97 = VectorParameter() {group=00_SuperLevels_T2, parameterName=Paint_Color4B, defaultValue=<rgba 0.0044 0.0123 0.1198 1>}
  n98 = VectorParameter() {group=00_SuperLevels_T2, parameterName=Paint_Color4, defaultValue=<rgba 0.0794 0.2231 0.8021 1>}
  n99 = NamedRerouteUsage()
  n100 = Constant2Vector()
  n101 = TextureSampleParameter2D!(
   UVs=Panner(Coordinate=Add(A=@n99, B=@n100), Speed=NamedRerouteUsage()) {speedX=-0.1, speedY=-0.2}
   "DDX(UVs)"=NamedRerouteUsage()
   "DDY(UVs)"=NamedRerouteUsage()
   Apply View MipBias
  ) {
   group=00_SuperLevels_T2
   mipValueMode=TMVM_Derivative
   parameterName=T2_Paint
   samplerSource=SSM_Wrap_WorldGroupSettings
   samplerType=SAMPLERTYPE_LinearColor
   texture=/BRCosmetics/Characters/Player/Male/Large/Bodies/M_LRG_BengalBasher_Suit/Skins/SL/Textures/T_S28_SL2_StrokesLayer
  }
  n102 = NamedRerouteUsage()
  n103 = Constant2Vector() {g=-0.31, r=0.49}
  n104 = NamedRerouteUsage()
  n105 = TextureSampleParameter2D(
   UVs=Panner(Coordinate=Add(A=@n102, B=@n103), Speed=Multiply(A=@n104) {constB=-0.5}) {speedX=-0.13, speedY=-0.22}
   "DDX(UVs)"=NamedRerouteUsage()
   "DDY(UVs)"=NamedRerouteUsage()
  ) {
   group=00_SuperLevels_T2
   mipValueMode=TMVM_Derivative
   parameterName=T2_Paint
   samplerSource=SSM_Wrap_WorldGroupSettings
   samplerType=SAMPLERTYPE_LinearColor
   texture=&T_S28_SL2_StrokesLayer
  }
  n106 = LinearInterpolate(
   A=LinearInterpolate(A=@n96.RGB, B=LinearInterpolate(A=@n97.RGB, B=@n98.RGB, Alpha=@n101.R), Alpha=@n101.G)
   B=LinearInterpolate(
    A=VectorParameter() {group=00_SuperLevels_T2, parameterName=Paint_Color3B, defaultValue=<rgba 0.2865 0.0403 0.0373 1>}.RGB
    B=VectorParameter() {group=00_SuperLevels_T2, parameterName=Paint_Color3, defaultValue=<rgba 0.3802 0.1818 0.1932 1>}.RGB
    Alpha=@n105.R
   )
   Alpha=@n105.G
  )
  n107 = NamedRerouteUsage()
  n108 = Constant2Vector() {g=0.64, r=-0.29}
  n109 = NamedRerouteUsage()
  n110 = TextureSampleParameter2D(
   UVs=Panner(Coordinate=Add(A=@n107, B=@n108), Speed=Multiply(A=@n109) {constB=0.75}) {speedX=-0.16, speedY=-0.24}
   "DDX(UVs)"=NamedRerouteUsage()
   "DDY(UVs)"=NamedRerouteUsage()
  ) {
   group=00_SuperLevels_T2
   mipValueMode=TMVM_Derivative
   parameterName=T2_Paint
   samplerSource=SSM_Wrap_WorldGroupSettings
   samplerType=SAMPLERTYPE_LinearColor
   texture=&T_S28_SL2_StrokesLayer
  }
  n111 = LinearInterpolate(
   A=VectorParameter() {group=00_SuperLevels_T2, parameterName=Paint_Color2B, defaultValue=<rgba 0.0026 0.0277 0.125 1>}.RGB
   B=VectorParameter() {group=00_SuperLevels_T2, parameterName=Paint_Color2, defaultValue=<rgba 0 0.3997 0.9115 1>}.RGB
   Alpha=@n110.R
  )
  n112 = VectorParameter() {group=00_SuperLevels_T2, parameterName=Paint_Color1B, defaultValue=<rgba 0.474 0.0858 0 1>}
  n113 = VectorParameter() {group=00_SuperLevels_T2, parameterName=Paint_Color1, defaultValue=<rgba 0.8854 0.3305 0.0138 1>}
  n114 = NamedRerouteUsage()
  n115 = Constant2Vector() {g=-0.67, r=-0.23}
  n116 = NamedRerouteUsage()
  n117 = TextureSampleParameter2D(
   UVs=Panner(Coordinate=Add(A=@n114, B=@n115), Speed=Multiply(A=@n116) {constB=2.125}) {speedX=-0.2, speedY=-0.28}
   "DDX(UVs)"=NamedRerouteUsage()
   "DDY(UVs)"=NamedRerouteUsage()
  ) {
   group=00_SuperLevels_T2
   mipValueMode=TMVM_Derivative
   parameterName=T2_Paint
   samplerSource=SSM_Wrap_WorldGroupSettings
   samplerType=SAMPLERTYPE_LinearColor
   texture=&T_S28_SL2_StrokesLayer
  }
  n118 = LinearInterpolate(
   A=LinearInterpolate(A=LinearInterpolate(A=@n106, B=@n111, Alpha=@n110.G), B=LinearInterpolate(A=@n112.RGB, B=@n113.RGB, Alpha=@n117.R), Alpha=@n117.G)
   B=@n96.RGB
   Alpha=NamedRerouteUsage()
  )
  n119 = ScalarParameter() {defaultValue=0.15, group=00_SuperLevels_T2, parameterName=PaintFresnelEXP}
  n120 = ScalarParameter() {group=00_SuperLevels_T2, parameterName=PaintFresnelEdge}
  n121 = Multiply(
   A=OneMinus(None=Fresnel(ExponentIn=@n119, BaseReflectFractionIn=@n120))
   B=ScalarParameter() {defaultValue=4, group=00_SuperLevels_T2, parameterName=Brightness}
  )
  n122 = Fresnel(ExponentIn=ScalarParameter() {defaultValue=5, group=00_SuperLevels_T2, parameterName=FresnelRimExp}) {baseReflectFraction=0, exponent=13}
  n123 = VectorParameter() {group=00_SuperLevels_T2, parameterName=FresnelRimColor, defaultValue=<rgba 0.503 0.2322 0.0572 1>}
  n124 = Multiply(A=NamedRerouteUsage(), B=Add(A=Multiply(A=@n118, B=@n121), B=Multiply(A=@n122, B=@n123.RGB)))
  n125 = Constant()
  n126 = StaticSwitchParameter(
   True=@n95.MaterialAttributes
   False=BlendMaterialAttributes(
    A=@n95.MaterialAttributes
    B=SetMaterialAttributes!(MaterialAttributes, EmissiveColor=@n124, BaseColor=@n125, Normal=@n95.Normal) {attributeSetTypes [#2, #1, #4]}
    Alpha=NamedRerouteUsage()
   ) {vertexAttributeBlendType=UseA}
  ) {group=00_SuperLevels, parameterName=SL_DisablePrimaryLayer, sortPriority=40}
  n127 = GetMaterialAttributes(None=@n126) {attributeGetTypes [#1]}
  n128 = Multiply(
   A=Desaturation(None=@n127.BaseColor, Fraction=Constant() {r=0.85})
   B=VectorParameter() {group=00_SuperLevels_T2, parameterName=SL_T2_AccentColor, defaultValue=<rgba 0.3707 0.8697 1 1>}.RGB
  )
  n129 = Constant() {r=1}
  n130 = ScalarParameter() {defaultValue=0.17, group=00_SuperLevels_T2, parameterName=SL_T2_AccentRoughness}
  n131 = StaticSwitchParameter(
   True=@n126
   False=BlendMaterialAttributes(
    A=@n126
    B=SetMaterialAttributes!(MaterialAttributes=@n127.MaterialAttributes, BaseColor=@n128, Metallic=@n129, Roughness=@n130) {attributeSetTypes [#1, #6, #3]}
    Alpha=NamedRerouteUsage()
   )
  ) {group=00_SuperLevels, parameterName=SL_DisableAccentLayer, sortPriority=42}
  n132 = GetMaterialAttributes(None=@n131) {attributeGetTypes [#4]}
  n133 = ScalarParameter() {defaultValue=2, group=00_SuperLevels_T2, parameterName=SL_T2_SecondaryEmissiveExponent}
  n134 = VectorParameter() {group=00_SuperLevels_T2, parameterName=SL_T2_SecondaryEmissiveColor, defaultValue=<rgba 0.1892 0.7257 1 1>}
  n135 = SetMaterialAttributes!(
   MaterialAttributes
   BaseColor=VectorParameter() {group=00_SuperLevels_T2, parameterName=SL_T2_SecondaryColor, defaultValue=<rgba 0.0017 0.0111 0.03 0>}.RGB
   Normal=@n132.Normal
   EmissiveColor=Multiply(A=Fresnel(ExponentIn=@n133) {baseReflectFraction=0}, B=@n134.RGB)
   Roughness=ScalarParameter() {defaultValue=0.55, group=00_SuperLevels_T2, parameterName=SL_T2_SecondaryRoughness, sliderMax=1}
  ) {attributeSetTypes [
    #1
    #4
    #2
    #3
   ]}
  n136 = NamedRerouteUsage()
  NamedRerouteDeclaration(None=StaticSwitchParameter(True=@n131, False=BlendMaterialAttributes(A=@n132.MaterialAttributes, B=@n135, Alpha=@n136)) {group=00_SuperLevels, parameterName=SL_DisableSecondaryLayer, sortPriority=41}) {name=T2_MA, nodeColor=<rgba 1 0 0.6328 1>}
  n137 = Multiply(A=NamedRerouteUsage()) {constB=1.8}
  n138 = VectorParameter() {group=00_SuperLevels_T1, parameterName=SL_T1_SecondaryColor, defaultValue=<rgba 0.0569 0.0528 0.176 1>}
  n139 = Constant()
  n140 = GetMaterialAttributes(None=NamedRerouteUsage()) {attributeGetTypes [#2, #4]}
  n141 = NamedRerouteUsage()
  n142 = Time()
  n143 = Subtract(A=@n141, B=Multiply(A=Constant2Vector() {g=-0.0115, r=-0.0024}, B=@n142))
  n144 = TextureSample!(UVs=@n143, Tex, "DDX(UVs)"=NamedRerouteUsage(), "DDY(UVs)"=NamedRerouteUsage(), Apply View MipBias) {
   mipValueMode=TMVM_Derivative
   samplerSource=SSM_Wrap_WorldGroupSettings
   samplerType=SAMPLERTYPE_Normal
   texture=/BRCosmetics/Characters/Player/Male/Large/Bodies/M_LRG_BengalBasher_Suit/Skins/SL/Textures/T_S28_SL1_BG_N
  }
  n145 = Subtract(A=@n141, B=Multiply(A=@n142, B=Constant2Vector() {g=0.0145, r=0.003}))
  n146 = TextureSample(UVs=@n145, "DDX(UVs)"=NamedRerouteUsage(), "DDY(UVs)"=NamedRerouteUsage()) {
   mipValueMode=TMVM_Derivative
   samplerSource=SSM_Wrap_WorldGroupSettings
   samplerType=SAMPLERTYPE_Normal
   texture=&T_S28_SL1_BG_N
  }
  n147 = MaterialFunctionCall(
   BaseNormal=@n140.Normal
   AdditionalNormal=LinearInterpolate(
    A=MaterialFunctionCall(BaseNormal=@n144.RGB, AdditionalNormal=@n146.RGB) {materialFunction=&BlendAngleCorrectedNormals}.Result
    B=Constant3Vector() {constant=<rgba 0 0 1 0>}
    Alpha=NamedRerouteUsage()
   )
  ) {materialFunction=&BlendAngleCorrectedNormals}
  n148 = MaterialFunctionCall(BaseNormal=@n147.Result, AdditionalNormal=VertexNormalWS()) {materialFunction=&BlendAngleCorrectedNormals}
  n149 = Fresnel(
   ExponentIn=ScalarParameter() {defaultValue=3, group=00_SuperLevels_T1, parameterName=FresnelPower}
   BaseReflectFractionIn=ScalarParameter() {defaultValue=0.1, group=00_SuperLevels_T1, parameterName=EmissivePassThrough}
   Normal=NamedRerouteDeclaration(None=Normalize(VectorInput=@n148.Result)) {name=T1 Blended Normal Map, nodeColor=<rgba 0.4297 0 1 1>}
  ) {baseReflectFraction=0.3}
  n150 = ScalarParameter() {defaultValue=5, group=00_SuperLevels_T1, parameterName=SL_PaperTextureScale}
  n151 = NamedRerouteUsage()
  n152 = Constant2Vector() {g=-0.0077, r=-0.0024}
  n153 = Time()
  n154 = ScalarParameter() {defaultValue=2, group=00_SuperLevels_T1, parameterName=SL_PaperTextureScale2}
  n155 = TextureSample(
   UVs=Add(A=Multiply(A=@n150, B=@n151) {constB=6}, B=Multiply(A=@n152, B=@n153))
   "DDX(UVs)"=Multiply(A=@n154, B=NamedRerouteUsage())
   "DDY(UVs)"=Multiply(A=NamedRerouteUsage(), B=@n154)
  ) {
   mipValueMode=TMVM_Derivative
   samplerSource=SSM_Wrap_WorldGroupSettings
   samplerType=SAMPLERTYPE_Normal
   texture=/Game/Environments/Asteria/Sets/Dojo/Materials/RicepaperMaterial/Pattern_PaperGrain_NM
  }
  n156 = ComponentMask(None=NamedRerouteUsage()) {g=true}
  n157 = ScalarParameter() {defaultValue=0.25, group=00_SuperLevels_T1, parameterName=Warp Intensity}
  n158 = NamedRerouteUsage()
  n159 = NamedRerouteUsage()
  n160 = TextureSample(UVs=Add(A=Multiply(A=@n156, B=@n157), B=@n143), "DDX(UVs)"=@n158, "DDY(UVs)"=@n159) {
   mipValueMode=TMVM_Derivative
   samplerSource=SSM_Wrap_WorldGroupSettings
   samplerType=SAMPLERTYPE_LinearColor
   texture=&T_S28_SL1_BG_A
  }
  n161 = TextureSample(UVs=@n145, "DDX(UVs)"=@n158, "DDY(UVs)"=@n159) {
   mipValueMode=TMVM_Derivative
   samplerSource=SSM_Wrap_WorldGroupSettings
   samplerType=SAMPLERTYPE_LinearColor
   texture=&T_S28_SL1_BG_A
  }
  n162 = Multiply(A=@n160.R, B=@n161.R)
  n163 = Add(
   A=Multiply(A=OneMinus(None=@n155.B), B=ScalarParameter() {defaultValue=15, group=00_SuperLevels_T1, parameterName=SL_PaperNormalDistortInfluence})
   B=Power(Base=@n162, Exp=ScalarParameter() {defaultValue=0.45, group=00_SuperLevels_T1, parameterName=SL_PatternPower}) {constExponent=1}
  )
  n164 = NamedRerouteUsage()
  n165 = NamedRerouteDeclaration(None=SphereMask(
    A=ComponentMask(None=@n164) {g=true, r=true}
    B=Constant2Vector()
    Radius=ScalarParameter() {defaultValue=7, group=00_SuperLevels_T1, parameterName=SL_T1_CylUV_AxisMaskRadius}
    Hardness=ScalarParameter() {defaultValue=0.2, group=00_SuperLevels_T1, parameterName=SL_T1_CylUV_AxisMaskHardness, sliderMax=1}
   ) {attenuationRadius=15, hardnessPercent=50}) {name=T1 CylUV Axis Mask, nodeColor=<rgba 0.5703 0 1 1>}
  n166 = CurveAtlasRowParameter(CurveTime=Multiply(A=Saturate(None=@n163), B=OneMinus(None=@n165))) {
   atlas=/BRCosmetics/Characters/Player/Male/Large/Bodies/M_LRG_BengalBasher_Suit/Skins/SL/Textures/CA_S28_SuperLevels
   curve=/BRCosmetics/Characters/Player/Male/Large/Bodies/M_LRG_BengalBasher_Suit/Skins/SL/Textures/C_S28_SL1_PrimaryBG_Alt
   defaultValue=2
   group=00_SuperLevels_T1
   parameterName=BG_Gradient
  }
  n167 = MaterialFunctionCall(
   LDR Input=Multiply(A=@n149, B=@n166)
   HDR Intensity Multiplier=ScalarParameter() {defaultValue=5, group=00_SuperLevels_T1, parameterName=SL_BG_ColorBrightness}
  ) {materialFunction=&DeriveHDRfromLDR}
  n168 = SetMaterialAttributes!(
   MaterialAttributes
   BaseColor=Constant()
   Metallic=@n139
   Normal=@n147.Result
   EmissiveColor=Multiply(A=NamedRerouteUsage(), B=Multiply(A=@n167.RemappedRange) {constB=0.95})
   Roughness=ScalarParameter() {defaultValue=1, group=00_SuperLevels_T1, parameterName=sL_T1_PrimaryRoughness, sliderMax=1}
   Specular=@n139
  ) {attributeSetTypes [
    #1
    #6
    #4
    #2
    #3
    #7
   ]}
  n169 = NamedRerouteUsage()
  n170 = GetMaterialAttributes(None=StaticSwitchParameter(
    True=@n140.MaterialAttributes
    False=BlendMaterialAttributes(A=@n140.MaterialAttributes, B=@n168, Alpha=@n169) {vertexAttributeBlendType=UseA}
   ) {group=00_SuperLevels, parameterName=SL_DisablePrimaryLayer, sortPriority=40}) {attributeGetTypes [
    #3
    #4
    #7
    #1
   ]}
  n171 = Multiply(
   A=Desaturation(None=@n170.BaseColor, Fraction=Constant() {r=0.95})
   B=ScalarParameter() {defaultValue=15, parameterName=SL_T1_Secondary_BaseColorInfluence}
  ) {constB=1.5}
  n172 = ScalarParameter() {defaultValue=4.5, group=00_SuperLevels_T1, parameterName=SL_T1_SecondaryEmissiveExponent}
  n173 = SetMaterialAttributes!(
   MaterialAttributes=@n170.MaterialAttributes
   BaseColor=Multiply(A=Multiply(A=@n137, B=@n138.RGB), B=Saturate(None=@n171))
   EmissiveColor=Multiply(
    A=VectorParameter() {group=00_SuperLevels_T1, parameterName=SL_T1_SecondaryFresnelColor, defaultValue=<rgba 1.54 0.248 8 1>}.RGB
    B=Fresnel(ExponentIn=@n172) {baseReflectFraction=0, exponent=7}
   )
   Roughness=@n170.Roughness
   Normal=@n170.Roughness
  ) {attributeSetTypes [
    #1
    #2
    #3
    #4
   ]}
  n174 = StaticSwitchParameter(True=@n170.MaterialAttributes, False=BlendMaterialAttributes(A=@n170.MaterialAttributes, B=@n173, Alpha=NamedRerouteUsage())) {group=00_SuperLevels, parameterName=SL_DisableSecondaryLayer, sortPriority=41}
  n175 = GetMaterialAttributes(None=@n174) {attributeGetTypes [#1]}
  n176 = Constant() {r=0.85}
  n177 = SetMaterialAttributes!(
   MaterialAttributes=@n175.MaterialAttributes
   BaseColor=Multiply(
    A=Desaturation(None=@n175.BaseColor, Fraction=@n176)
    B=VectorParameter() {group=00_SuperLevels_T1, parameterName=SL_T1_AccentColor, defaultValue=<rgba 1 0.8152 0.2044 1>}.RGB
   )
   Metallic=Constant() {r=1}
   Roughness=ScalarParameter() {defaultValue=0.17, group=00_SuperLevels_T1, parameterName=SL_T1_AccentRoughness}
  ) {attributeSetTypes [#1, #6, #3]}
  n178 = NamedRerouteUsage()
  NamedRerouteDeclaration(None=StaticSwitchParameter(True=@n174, False=BlendMaterialAttributes(A=@n174, B=@n177, Alpha=@n178)) {group=00_SuperLevels, parameterName=SL_DisableAccentLayer, sortPriority=42}) {name=T1_MA, nodeColor=<rgba 1 0.3516 0 1>}
  n179 = NamedRerouteUsage()
  n180 = Multiply(
   A=MaterialFunctionCall(Vector or UVs=ComponentMask(None=@n179) {g=true, r=true}) {materialFunction=&VectorToRadialValue}.Vector Converted to Angle
   B=ScalarParameter() {defaultValue=1, group=00_SuperLevels_T1, parameterName=SL_Radial Tiling}
  ) {constB=2}
  n181 = ComponentMask(None=PreSkinnedLocalBounds().Extents) {b=true}
  n182 = ComponentMask(None=Divide(A=@n179, B=NamedRerouteDeclaration(None=@n181) {name=SL_PreskinnedLocalBoundsZ, nodeColor=<rgba 0 1 0.3203 1>})) {b=true}
  n183 = Constant() {r=1}
  n184 = ScalarParameter() {defaultValue=1, group=00_SuperLevels_T1, parameterName=CylindricalUVs_VerticalScale}
  NamedRerouteDeclaration(None=Multiply(A=AppendVector(A=@n180, B=@n182), B=AppendVector(A=@n183, B=@n184))) {name=CylUV, nodeColor=<rgba 0.6875 0 1 1>}
  n185 = Multiply(A=Subtract(A=Desaturation(None=@n106)) {constB=0.09}) {constB=5.4}
  NamedRerouteDeclaration(None=Power(Base=Saturate(None=@n185), Exp=ScalarParameter() {defaultValue=1, group=00_SuperLevels_T2, parameterName=Sl_T2_EmissiveMaskExponent}) {constExponent=0.5}) {name=T2_EmissiveMaskVariation, nodeColor=<rgba 0.9922 0 1 1>}
  NamedRerouteDeclaration(None=@n46) {name=T3_Color, nodeColor=<rgba 0 0.4141 1 1>}
  NamedRerouteDeclaration(None=@n54.R) {name=T3_InnerPattern, nodeColor=<rgba 0 1 0.9297 1>}
  n186 = FunctionInput() {inputName=M, sortPriority=10}
  NamedRerouteDeclaration(None=ComponentMask(None=@n186) {r=true}) {name=AO, nodeColor=<rgba 0 1 0.1094 1>}
  n187 = NamedRerouteUsage()
  n188 = Multiply(A=SphereMask(
    A=ComponentMask(None=@n187) {g=true, r=true}
    B=Constant2Vector()
    Radius=ScalarParameter() {defaultValue=7, parameterName=SL_CylUV_AxisMaskRadius}
    Hardness=ScalarParameter() {defaultValue=0.4, parameterName=SL_CylUV_AxisMaskHardness, sliderMax=1}
   ) {attenuationRadius=15, hardnessPercent=50}) {constB=2}
  NamedRerouteDeclaration(None=Saturate(None=Subtract(A=@n188, B=@n117.G))) {name=T2 CylUV Axis Mask, nodeColor=<rgba 1 0 0.1641 1>}
  n189 = ComponentMask(None=@n49) {g=true, r=true}
  n190 = ScalarParameter() {defaultValue=0.2, group=00_SuperLevels_T3, parameterName=SL_T3_RefractIntensity}
  n191 = TextureCoordinate()
  n192 = TextureSample(UVs=Add(A=Multiply(A=@n189, B=@n190), B=@n52), "DDX(UVs)"=DDX(Value=@n191), "DDY(UVs)"=DDY(Value=@n191)) {
   mipValueMode=TMVM_Derivative
   samplerSource=SSM_Wrap_WorldGroupSettings
   samplerType=SAMPLERTYPE_LinearColor
   texture=&T_S28_SL1_BG_A
  }
  NamedRerouteDeclaration(None=@n192.R) {name=T3_InnerPatternCleanRefraction, nodeColor=<rgba 1 0 0.1406 1>}
  n193 = ScalarParameter() {defaultValue=1, group=00_SuperLevels, parameterName=SL_PrimaryGmapInfluence, sliderMax=1}
  n194 = LinearInterpolate(B=ComponentMask(None=@n186) {g=true}, Alpha=@n193) {constA=1}
  NamedRerouteDeclaration(None=@n194) {name=Gmap Influence without brightness correction, nodeColor=<rgba 0 1 0.1562 1>}
  NamedRerouteDeclaration(None=Multiply(A=Add(A=@n193), B=@n194)) {name=Gmap Influence, nodeColor=<rgba 0.0078 0 1 1>}
  NamedRerouteDeclaration(None=@n162) {name=T1_EmissiveMaskModulation, nodeColor=<rgba 0 1 0.625 1>}
  n195 = NamedRerouteUsage()
  n196 = Constant2Vector() {g=1.5, r=1}
  NamedRerouteDeclaration(None=Multiply(A=Multiply(A=@n195, B=@n196), B=ScalarParameter() {defaultValue=2, group=00_SuperLevels_T2, parameterName=PaintUVScale})) {name=PaintStrokeUV, nodeColor=<rgba 0 0.8359 1 1>}
  n197 = ComponentMask(None=VectorParameter() {group=00_SuperLevels_T2, parameterName=PanDirectionRG, defaultValue=<rgba -0.5 -1 0 1>}.RGB) {g=true, r=true}
  NamedRerouteDeclaration(None=Multiply(A=Normalize(VectorInput=@n197), B=ScalarParameter() {defaultValue=0.1, group=00_SuperLevels_T2, parameterName=PanSpeed})) {name=PaintPanSpeed, nodeColor=<rgba 0 1 0.5547 1>}
  NamedRerouteUsage()
  PixelNormalWS()
  PreSkinnedLocalBounds()
  n198 = TextureSample!(UVs=TextureCoordinate() {uTiling=8, vTiling=8}, Tex, Apply View MipBias) {
   samplerSource=SSM_Wrap_WorldGroupSettings
   samplerType=SAMPLERTYPE_Normal
   texture=&Pattern_PaperGrain_NM
  }
  Reroute(None=LinearInterpolate(
    A=@n170.Normal
    B=MaterialFunctionCall(BaseNormal=@n170.Normal, AdditionalNormal=@n198.RGB) {materialFunction=&BlendAngleCorrectedNormals}.Result
    Alpha=ScalarParameter() {defaultValue=0.25, group=00_SuperLevels_T1, parameterName=SL_T1_SecondaryNormalDetail, sliderMax=2, sliderMin=-2}
   ))
  n199 = ScalarParameter() {group=00_SuperLevels_T1, parameterName=SL_T1_SecondaryRoughnessVariationAmt, sliderMax=2, sliderMin=-2}
  Reroute(None=Add(A=@n170.Roughness, B=Multiply(B=@n199)))
  Saturate(None=@n149)
  VectorParameter() {group=00_SuperLevels_T1, parameterName=SL_CylindricalProj_Tiling, channelNames {a="", b="", g=V, r=U}, defaultValue=<rgba 1 200 0 1>}
 ]}
