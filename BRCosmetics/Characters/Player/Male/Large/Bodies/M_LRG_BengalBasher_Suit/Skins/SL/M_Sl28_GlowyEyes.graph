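# Skeletal mesh material for glowing eyes on a skin of the large male Bengal Basher suit character.
# A Fresnel term blends a very bright HDR violet core color toward a dark blue-purple outer color.
# Both colors and the Fresnel exponent are parameters, with the tiny default exponent pushing the blend sharply toward the outer color away from head-on view.
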Material /BRCosmetics/Characters/Player/Male/Large/Bodies/M_LRG_BengalBasher_Suit/Skins/SL/M_Sl28_GlowyEyes
graph {
 properties {bUsedWithMeshDeformer=true, bUsedWithSkeletalMesh=true, pixelDepthOffsetMode=PDOM_Legacy, usedShadingModels=Default Lit}
 nodes [
  Constant()
  Constant() {r=0.15}
  LinearInterpolate(
   A=VectorParameter() {parameterName=GlowEyesCore, defaultValue=<rgba 337.8899 0 800 1>}.RGB
   B=VectorParameter() {parameterName=GlowEyesOutside, defaultValue=<rgba 0.0275 0 0.1562 1>}.RGB
   Alpha=Fresnel(ExponentIn=ScalarParameter() {defaultValue=0.004, parameterName=exc}) {baseReflectFraction=0, exponent=0.1}
  )
 ]
}
